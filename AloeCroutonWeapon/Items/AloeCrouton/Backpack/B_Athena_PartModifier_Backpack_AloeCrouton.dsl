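; Backpack part modifier for the AloeCrouton weapon that shows its charge level through material parameters on the backpack mesh.
; On pawn change it caches Charge and MaxCharge from the attribute set, falling back to PlayerState simulated values for sim proxies.
; A watched cosmetic stat drives a timeline animating charge and emissive flash, with a gain sound for the autonomous-proxy player while unequipped.
; PostResIn reapplies cached values to guard against parameter resets.
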(event EventOnPawnAndPartChanged
  (bind _returnvalue (CustomCharacter|GetSkeletalMeshForPartOnPlayerPawn))
  (|CallParentFunction)
  (Utilities|FlowControl|Sequence
    (:then_0
      (bind _returnvalue_2 (GameplayAttributes|GetAttributeSet (Ability|GetAbilitySystemComponent (CustomCharacter|GetAssociatedPlayerPawn)) "/Script/AloeCroutonWeaponRuntime.FortChargingSet_AloeCrouton"))
      (bind _asfort_charging_set_aloe_crouton (Utilities|Casting|CastToFortChargingSet_AloeCrouton _returnvalue_2)
        (:then
          ;[1 1 1] Cache charge values
          (bind (_charge_minimum _charge_maximum _charge_biscurrentclamped _charge_bisbaseclamped _charge_bshouldclampbase _charge_unclampedbasevalue _charge_unclampedcurrentvalue _charge_bshouldscalewithmax _charge_basevalue _charge_currentvalue) (Utilities|Struct|BreakFortGameplayAttributeData (Variables|/Script/AloeCroutonWeaponRuntime.FortChargingSet_AloeCrouton|GetCharge_Minimum _asfort_charging_set_aloe_crouton)))
          (Variables|Charge|Values|SetCachedChargeValue _charge_currentvalue)
          (bind (_maxcharge_minimum _maxcharge_maximum _maxcharge_biscurrentclamped _maxcharge_bisbaseclamped _maxcharge_bshouldclampbase _maxcharge_unclampedbasevalue _maxcharge_unclampedcurrentvalue _maxcharge_bshouldscalewithmax _maxcharge_basevalue _maxcharge_currentvalue) (Utilities|Struct|BreakFortGameplayAttributeData (Variables|/Script/AloeCroutonWeaponRuntime.FortChargingSet_AloeCrouton|GetMaxCharge_Minimum _asfort_charging_set_aloe_crouton)))
          (Variables|Charge|Values|SetCachedMaxChargeValue _maxcharge_currentvalue))
        (:CastFailed
          ;[1 1 1] Sim proxies won't have access to the attribute set, but they can get the simulated attribute values
          (bind _asfort_player_state_zone (Utilities|Casting|CastToFortPlayerStateZone (Variables|/Script/Engine.Pawn|GetPlayerState (CustomCharacter|GetAssociatedPlayerPawn))))
          (bind _output_get (Variables|Default|SetAssociatedPlayerState _asfort_player_state_zone))
          (Utilities|IsValid _output_get
            (:Is%20Valid
              (Variables|Charge|Values|SetCachedChargeValue (GadgetAttributeSets|GetSimulatedAttributeValue (Variables|Default|GetAssociatedPlayerState) (Variables|Charge|Attributes|GetChargeAttribute) 0.0 :!class /Script/FortniteGame.FortPlayerStateZone))
              (Variables|Charge|Values|SetCachedMaxChargeValue (GadgetAttributeSets|GetSimulatedAttributeValue (Variables|Default|GetAssociatedPlayerState) (Variables|Charge|Attributes|GetMaxChargeAttribute) 0.0 :!class /Script/FortniteGame.FortPlayerStateZone)))))))
    (:then_1
      ;[1 1 1] When Charge changes, update the material parameters and cache the value
      (bind (_statname _statvalue) (Cosmetics|WatchCosmeticStat (CustomCharacter|GetAssociatedPlayerPawn) "(TagName=\"StatManager.AloeCroutonCharge\")" (Math|Float|Truncate (Variables|Charge|Values|GetCachedChargeValue)) false)
        (:OnStatChanged
          (if (< (Variables|Charge|Values|GetCachedChargeValue) (Math|Conversions|ToFloat(Integer) _statvalue))
            (CallFunction|OnChargeAdded)
            (bind (_direction _ramp _flash) (Timeline|TL_Charge PlayFromStart)
              (:Update
                (Rendering|Material|SetScalarParameterValueonMaterials _returnvalue "AC_Charges_Current" (+ (* _ramp (- (Math|Conversions|ToFloat(Integer) _statvalue) (Variables|Charge|Values|GetCachedChargeValue))) (Variables|Charge|Values|GetCachedChargeValue)))
                (Rendering|Material|SetScalarParameterValueonMaterials _returnvalue "AC_Charges_EmissiveIntensity" _flash))
              (:Finished
                (Variables|Charge|Values|SetCachedChargeValue (Math|Conversions|ToFloat(Integer) _statvalue))))
            (else
              (Timeline|TL_Charge PlayFromStart))))))))

(event EventOnPostResIn
  (|CallParentFunction)
  ;[1 1 1] Always get the Charge and MaxCharge values and update material parameters on PostResIn (to prevent parameter values accidentally resetting in certain cases)
  (CallFunction|UpdateMaterialParams (Variables|Charge|Values|GetCachedChargeValue) (Variables|Charge|Values|GetCachedMaxChargeValue)))

(event Custom|OnChargeAdded
  (bind _returnvalue (CustomCharacter|GetAssociatedPlayerPawn))
  (if (Utilities|Enum|Equal(Enum) (Networking|GetLocalRole _returnvalue) "ROLE_AutonomousProxy")
    ;[1 1 1] Only play the sound if the backpack is showing (when the weapon is NOT equipped)
    (if (not (GameplayTags|HasMatchingGameplayTag _returnvalue (Variables|Default|GetEquippedTag)))
      (Audio|PlaySound2D (Variables|Sound|GetFullChargeSound) 1.0 1.0 0.0 0 _returnvalue))))

(fn UpdateMaterialParams
  (params (ChargeValue :type Float%20%28double-precision%29) (MaxChargeValue :type Float%20%28double-precision%29))
  (bind _returnvalue (CustomCharacter|GetSkeletalMeshForPartOnPlayerPawn))
  (Rendering|Material|SetScalarParameterValueonMaterials _returnvalue "AC_Charges_Current" (Variables|GetChargeValue))
  (Rendering|Material|SetScalarParameterValueonMaterials _returnvalue "AC_Charges_Max" (Variables|GetMaxChargeValue)))

(fn ConstructionScript
  (|CallParentFunction))
